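# Translucent, unlit, two-sided material for the DarkStance Pater character's skirt FX mesh, drawing wispy, scrolling smoke-trail strands, green by default.
# Color blends two parameters along U, with smoke-texture-driven hue shift.
# Opacity combines noise-distorted trail textures, caustic erosion, U-end fades and depth fade.
# SplineThicken, sine waves along two local directions and camera offset build world position offset, with vertex red varying width, phase and UVs.
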Material /BRCosmetics/Characters/Player/Male/Medium/Bodies/M_MED_DarkStance_Pater/FX/M_DarkStancePater_SkirtFXMesh
graph {
 properties {
  bUsedWithMeshDeformer=true
  bUsedWithNiagaraMeshParticles=true
  bUsedWithNiagaraRibbons=true
  bUsedWithSkeletalMesh=true
  blendMode=BLEND_Translucent
  pixelDepthOffsetMode=PDOM_Legacy
  shadingModel=MSM_Unlit
  twoSided=true
  usedShadingModels=Unlit
 }
 nodes [
  n1 = LinearInterpolate(Alpha=TextureSample() {texture=&T_Inky_Smoke_Tile}.R) {constA=-0.02, constB=0.05}
  n2 = ScalarParameter() {defaultValue=-0.01, parameterName=Color offset}
  n3 = ComponentMask(None=TextureCoordinate()) {r=true}
  n4 = LinearInterpolate(
   A=VectorParameter() {parameterName=Color1, defaultValue=<rgba 0 0.7 0.1968 1>}.RGB
   B=VectorParameter() {parameterName=Color2, defaultValue=<rgba 0 1 0.2812 1>}.RGB
   Alpha=Power(Base=OneMinus(None=@n3), Exp=ScalarParameter() {defaultValue=1, parameterName=ColorRangePower}) {constExponent=1}
  )
  MaterialFunctionCall(Emissive=MaterialFunctionCall(Hue Shift Percentage=Add(A=@n1, B=@n2) {constB=0}, Texture=Multiply(A=@n4) {constB=0.5}) {materialFunction=&HueShift}.Result) {materialFunction=&MF_FX_EmissiveAdjustments}
  n5 = Transform(None=VectorParameter() {parameterName=WPO Direction1, defaultValue=<rgba 0 1 0 1>}.RGB) {transformSourceType=TRANSFORMSOURCE_Local}
  n6 = VertexColor()
  n7 = Multiply(A=TextureCoordinate() {uTiling=0.7}, B=LinearInterpolate(Alpha=@n6.R) {constA=0.7})
  n8 = Add(A=Time(), B=LinearInterpolate(Alpha=@n6.R) {constA=-0.5})
  n9 = Sine(None=Add(A=ComponentMask(None=@n7) {r=true}, B=Multiply(A=@n8) {constB=-0.3})) {period=0.5}
  Multiply(A=@n5, B=@n9)
  n10 = TextureCoordinate() {uTiling=0.8}
  n11 = AppendVector(A=VertexColor().R, B=Constant())
  n12 = Add(A=TextureCoordinate() {uTiling=0.8}, B=AppendVector(A=VertexColor().R, B=Constant()))
  n13 = Subtract(A=TextureSample(UVs=Panner(Coordinate=@n12) {speedX=0.1}) {samplerType=SAMPLERTYPE_Grayscale, texture=&LowResBlurredNoise}.R) {constB=0.5}
  n14 = Panner(Coordinate=Add(A=Add(A=@n10, B=@n11), B=Multiply(A=@n13) {constB=0.2}) {constB=0}) {speedX=-0.5}
  n15 = Max(A=Multiply(A=TextureSample(UVs=@n14) {samplerType=SAMPLERTYPE_Grayscale, texture=&T_Smoke_Trail_Soft}.R)) {constB=0}
  Multiply(A=@n15)
  n16 = Constant() {r=5}
  n17 = Constant() {r=25}
  n18 = LinearInterpolate(Alpha=VertexColor().R) {constA=0.7, constB=1.2}
  n19 = Constant() {r=10}
  n20 = VectorParameter() {parameterName=WPO Direction2, defaultValue=<rgba 1 0 0 1>}
  n21 = Add(
   A=Multiply(A=@n5, B=LinearInterpolate(Alpha=@n9))
   B=Multiply(A=LinearInterpolate(Alpha=@n9) {constA=0.3, constB=1.3}, B=Transform(None=@n20.RGB) {transformSourceType=TRANSFORMSOURCE_Local})
  )
  n22 = ScalarParameter() {defaultValue=2, parameterName=WPO Intensity}
  n23 = Multiply(A=ComponentMask(None=TextureCoordinate()) {r=true}) {constB=3}
  n24 = Add(
   A=MaterialFunctionCall(WidthBase=Multiply(A=@n17, B=@n18), WidthTip=Multiply(A=@n19), Expand U or V UV Channel=StaticBool()) {materialFunction=&SplineThicken}.WorldPositionOffset
   B=Multiply(A=Multiply(A=@n21, B=@n22), B=Saturate(None=@n23))
  )
  Reroute(None=Add(
    A=Add(A=@n24) {constB=0}
    B=MaterialFunctionCall(Offset Amount=@n16) {materialFunction=/Engine/Functions/Engine_MaterialFunctions02/WorldPositionOffset/CameraOffset}.Camera Offset
   ) {constB=0})
  n25 = VertexColor()
  n26 = Constant()
  n27 = Add(A=Add(A=TextureCoordinate() {uTiling=0.8, vTiling=0.7}, B=AppendVector(A=@n25.R, B=@n26)), B=Constant2Vector() {g=0.12})
  n28 = Multiply(A=@n13) {constB=0.15}
  n29 = TextureSample(UVs=Panner(Coordinate=Add(A=@n27, B=@n28) {constB=0}) {speedX=-0.5}) {samplerType=SAMPLERTYPE_LinearColor, texture=/BRCosmetics/Athena/Effects/Particles/Skydiving/NS_Contrail_CubeNinja/T_CubeNinja_Trail}
  n30 = TextureSample(UVs=@n14) {texture=/BRCosmetics/Weapons/FORT_Melee/Pickaxe_Future_Samurai_Summer_Male/FX/T_Pickaxe_Future_Samurai_Summer_Male_Trail_01}
  n31 = Multiply(A=LinearInterpolate(A=Power(Base=@n15) {constExponent=1.2}, B=Multiply(A=@n29.R), Alpha=@n30.R) {constAlpha=0.4}) {constB=1.2}
  n32 = Multiply(A=Saturate(None=Power(Base=@n31) {constExponent=1.2}), B=LinearInterpolate(Alpha=@n30.R) {constA=0.5})
  n33 = ComponentMask(None=TextureCoordinate()) {r=true}
  n34 = VertexColor()
  n35 = Constant()
  n36 = Panner(Coordinate=Add(A=TextureCoordinate() {vTiling=0.5}, B=AppendVector(A=@n34.R, B=@n35))) {speedX=-0.5}
  n37 = Power(Base=ComponentMask(None=TextureCoordinate()) {r=true}) {constExponent=1}
  n38 = LinearInterpolate(Alpha=Power(
    Base=TextureSample(UVs=@n36) {samplerType=SAMPLERTYPE_LinearGrayscale, texture=/Game/Global/Textures/Default/Caustics/T_Nosie_Caustics_Grayscale}.R
    Exp=LinearInterpolate(Alpha=@n37) {constA=-1, constB=0.5}
   )) {constA=-1}
  n39 = ComponentMask(None=TextureCoordinate()) {r=true}
  n40 = Multiply(A=Add(A=OneMinus(None=@n39)) {constB=-0.1}) {constB=3}
  n41 = Multiply(A=Saturate(None=Multiply(A=@n33)), B=Multiply(A=Saturate(None=@n38), B=Saturate(None=@n40)))
  n42 = Saturate(None=Multiply(A=Multiply(A=@n32, B=@n41) {constA=3}) {constB=3})
  Reroute(None=DepthFade(Opacity=Multiply(A=@n42)) {fadeDistanceDefault=5})
 ]
}
