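# Material function for the Styx character cosmetic, adding animated emissive to input material attributes and lowering specular by mask.
# Panned, distorted main, detail and macro noise use BumpOffset parallax, then blend with smoke and solid recolor within a Tech Art mask.
# Pre-skinned position drives masks and special heart (polar) and shoulder UVs.
# Static switches add edge glow or a sine-based Tyrian variant; a second output exposes UV distortion.
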MaterialFunction /BRCosmetics/Characters/Player/Male/Medium/Bodies/M_MED_DarkStance_Pater/Materials/MF_StyxCosmetic
graph {nodes [
  AppendVector(A=NamedRerouteUsage(), B=Constant())
  n1 = GetMaterialAttributes(None=NamedRerouteUsage()) {attributeGetTypes [#1, #2]}
  n2 = Multiply(A=@n1.EmissiveColor, B=ScalarParameter() {defaultValue=1, group=Styx, parameterName=OrgEmissiveIntensity})
  n3 = NamedRerouteUsage()
  n4 = StaticSwitchParameter(True=Add(A=NamedRerouteUsage(), B=@n3), False=@n3) {group=Styx, parameterName=UseEdgeGlow}
  n5 = NamedRerouteUsage()
  n6 = Multiply(A=@n5, B=ScalarParameter() {defaultValue=4, group=StyxMod, parameterName=FakeMetalicSinMulti}) {constB=2}
  n7 = Multiply(A=ConstantBiasScale(None=Sine(None=@n6)), B=@n5)
  n8 = Max(A=Desaturation(None=@n3), B=ScalarParameter() {defaultValue=0.1, group=Styx_Detail, parameterName=Styx_Detail_Noise_MinValue}) {constB=0.1}
  n9 = NamedRerouteUsage()
  n10 = NamedRerouteDeclaration(None=Multiply(A=Add(A=@n8, B=@n9), B=NamedRerouteUsage())) {name=DetailMask, nodeColor=<rgba 1 0.6797 0 1>}
  n11 = Multiply(
   A=Multiply(
    A=VectorParameter() {group=Styx_Detail, parameterName=StyxEmissive_DetailGlow, defaultValue=<rgba 0 5 0.4457 1>}.RGB
    B=StaticSwitchParameter(True=@n7, False=@n10) {group=StyxMod, parameterName=TyrianVersion}
   )
   B=ScalarParameter() {defaultValue=1, group=Styx_Detail, parameterName=DetailNoiseIntensity}
  )
  n12 = ScalarParameter() {defaultValue=1, group=Styx_Absoprtion_Noise, parameterName=MacrNoiseIntensity}
  n13 = LinearInterpolate(
   A=VectorParameter() {group=Styx_Absoprtion_Noise, parameterName=StyxEmissive_MacroNoise_A, defaultValue=<rgba 0 5 0.4457 1>}.RGB
   B=VectorParameter() {group=Styx_Absoprtion_Noise, parameterName=StyxEmissive_MacroNoise_B, defaultValue=<rgba 0 5 0.4457 1>}.RGB
   Alpha=NamedRerouteUsage()
  )
  n14 = ScalarParameter() {defaultValue=1.5, group=StyxMod, parameterName=SmokeGradientMultiMax}
  n15 = ScalarParameter() {defaultValue=0.7, group=StyxMod, parameterName=SmokeGradientMultiMin}
  n16 = NamedRerouteUsage()
  n17 = Saturate(None=Multiply(A=NamedRerouteUsage(), B=LinearInterpolate(A=@n14, B=@n15, Alpha=@n16) {constA=3}))
  n18 = LinearInterpolate(
   A=Add(A=Add(A=@n4, B=@n11), B=Multiply(A=@n12, B=@n13))
   B=VectorParameter() {group=StyxMod, parameterName=StyxSmokeEmissive, defaultValue=<rgba 0 0.0625 0.0056 1>}.RGB
   Alpha=NamedRerouteDeclaration(None=@n17) {name=SmokeAdjustedArea, nodeColor=<rgba 0.9219 0 1 1>}
  )
  n19 = VectorParameter() {group=StyxMod, parameterName=StyxSolidEmissive, defaultValue=<rgba 0.2 0.8 0.2 1>}
  n20 = ComponentMask(None=NamedRerouteUsage()) {b=true}
  n21 = ScalarParameter() {defaultValue=25, group=StyxMod, parameterName=RecolorSize}
  n22 = OneMinus(None=Saturate(None=Divide(A=@n20, B=@n21)))
  n23 = Saturate(None=Multiply(A=Saturate(None=@n22), B=ScalarParameter() {defaultValue=2, group=StyxMod, parameterName=RecolorScale}) {constB=2})
  n24 = SmoothStep(
   Min=ScalarParameter() {group=StyxMod, parameterName=TechArtMaskMin}
   Max=ScalarParameter() {defaultValue=1, group=StyxMod, parameterName=TechArtMaskMax}
   Value=Power(Base=NamedRerouteUsage(), Exp=ScalarParameter() {defaultValue=1, group=StyxMod, parameterName=TechArtMaskPow})
  )
  n25 = Multiply(
   A=Multiply(
    A=ScalarParameter() {defaultValue=1, group=Styx, parameterName=OverallEmissiveIntensity}
    B=LinearInterpolate(A=@n18, B=@n19.RGB, Alpha=@n23)
   )
   B=@n24
  )
  n26 = NamedRerouteUsage()
  FunctionOutput(None=SetMaterialAttributes!(
    MaterialAttributes=@n1.MaterialAttributes
    EmissiveColor=Add(A=@n2, B=@n25)
    Specular=LinearInterpolate(A=@n1.Specular, Alpha=@n26) {constB=0}
   ) {attributeSetTypes [#1, #2]})
  n27 = ScalarParameter() {group=Styx, parameterName=StyxNoise_SpeedX}
  n28 = ScalarParameter() {defaultValue=-0.1, group=Styx, parameterName=StyxNoise_SpeedY}
  n29 = TextureObjectParameter() {
   group=Styx
   parameterName=StyxDistortNormal
   samplerType=SAMPLERTYPE_Normal
   texture=/Game/Characters/Player/Male/Medium/Bodies/M_MED_DarkStance_Pater/Textures/T_StyxCosmetic_N
  }
  n30 = TextureSample(UVs=Panner(Coordinate=NamedRerouteUsage(), Speed=AppendVector(A=@n27, B=@n28)), Tex=@n29) {
   samplerSource=SSM_Wrap_WorldGroupSettings
   samplerType=SAMPLERTYPE_Normal
   texture=&T_StyxCosmetic_N
  }
  FunctionOutput(None=NamedRerouteDeclaration(None=ComponentMask(None=@n30.RGB) {g=true, r=true}) {name=Distortion_NormalNoise_A, nodeColor=<rgba 0.3828 0 1 1>}) {outputName=UV_Distort, sortPriority=1}
  Multiply(
   A=ComponentMask(None=NamedRerouteUsage()) {g=true, r=true}
   B=ScalarParameter() {defaultValue=0.01, group=Styx, parameterName=Styx_Distort_UV_Normal}
  )
  n31 = Panner(
   Coordinate=Add(A=NamedRerouteUsage(), B=NamedRerouteUsage())
   Speed=AppendVector(
    A=ScalarParameter() {group=Styx_Detail, parameterName=Styx_Detail_Noise_SpeedX}
    B=ScalarParameter() {defaultValue=-0.1, group=Styx_Detail, parameterName=Styx_Detail_Noise_SpeedY}
   )
  )
  n32 = TextureSampleParameter2D(UVs=@n31) {
   group=Styx_Detail
   parameterName=StyxNoiseDetail
   samplerType=SAMPLERTYPE_Masks
   texture=/Game/Characters/Player/Male/Medium/Bodies/M_MED_DarkStance_Pater/Textures/T_StyxCosmetic_DetailNoise_M
  }
  n33 = TextureSampleParameter2D(UVs=Multiply(A=NamedRerouteUsage(), B=ScalarParameter() {defaultValue=1, group=Styx, parameterName=MacroNoiseTilling})) {
   group=Styx_Detail
   parameterName=MacroNoise
   samplerType=SAMPLERTYPE_Grayscale
   texture=/BRCosmetics/Characters/Player/Male/Medium/Bodies/M_MED_DarkStance_Pater/Textures/T_StyxCosmetic_DetailNoise_HighQuality_M
  }
  n34 = LinearInterpolate(
   A=ScalarParameter() {defaultValue=-0.5, group=Styx_Absoprtion_Noise, parameterName=Styx_Absoprtion_Noise_Contrast_LOW}
   B=ScalarParameter() {defaultValue=2, group=Styx_Absoprtion_Noise, parameterName=Styx_Absoprtion_Noise_Contrast_High}
   Alpha=@n33.G
  )
  n35 = LinearInterpolate(
   A=ScalarParameter() {defaultValue=-0.5, group=Styx_Absoprtion_Noise, parameterName=Styx_Absoprtion_Noise_B_Contrast_LOW}
   B=ScalarParameter() {defaultValue=2, group=Styx_Absoprtion_Noise, parameterName=Styx_Absoprtion_Noise_B_Contrast_High}
   Alpha=@n33.G
  )
  Multiply(A=@n32.G, B=Add(A=Saturate(None=@n34), B=Saturate(None=@n35)))
  n36 = ScalarParameter() {defaultValue=2, group=Styx, parameterName=StyxMainNoiseTilling}
  n37 = MaterialFunctionCall(Float2=Multiply(A=@n36, B=TextureCoordinate())) {materialFunction=&BreakOutFloat2Components}
  n38 = NamedRerouteUsage()
  n39 = ScalarParameter() {defaultValue=1, group=StyxMod, parameterName=UVModAdd}
  n40 = MaterialFunctionCall(X=@n37.R, Y=Add(A=@n37.G, B=Multiply(A=@n38, B=@n39) {constB=4})) {materialFunction=&MakeFloat2}
  n41 = NamedRerouteUsage()
  n42 = NamedRerouteUsage()
  n43 = BumpOffset(
   Coordinate=StaticSwitch(True=@n40.Result, False=LinearInterpolate(A=@n40.Result, B=@n41, Alpha=@n42), Value=NamedRerouteUsage())
   Height=ScalarParameter() {defaultValue=-3, group=Styx, parameterName=StyxUVOffset02}
  )
  n44 = NamedRerouteUsage()
  n45 = NamedRerouteUsage()
  NamedRerouteDeclaration(None=StaticSwitch(True=@n43, False=LinearInterpolate(A=@n43, B=@n44, Alpha=@n45), Value=NamedRerouteUsage())) {name=UV_withParallax_B, nodeColor=<rgba 1 0.6562 0 1>}
  n46 = MaterialFunctionCall(Float2=Multiply(A=TextureCoordinate(), B=@n36)) {materialFunction=&BreakOutFloat2Components}
  n47 = Multiply(A=NamedRerouteUsage(), B=@n39) {constB=4}
  n48 = NamedRerouteDeclaration(None=MaterialFunctionCall(X=@n46.R, Y=Add(A=@n46.G, B=@n47)) {materialFunction=&MakeFloat2}.Result) {name=ModUV, nodeColor=<rgba 0 1 0.9297 1>}
  n49 = NamedRerouteUsage()
  n50 = NamedRerouteUsage()
  n51 = BumpOffset(
   Coordinate=StaticSwitch(True=@n48, False=LinearInterpolate(A=@n48, B=@n49, Alpha=@n50), Value=NamedRerouteUsage())
   Height=ScalarParameter() {defaultValue=-3, group=Styx, parameterName=StyxUVOffset01}
  )
  n52 = NamedRerouteUsage()
  n53 = NamedRerouteUsage()
  NamedRerouteDeclaration(None=StaticSwitch(True=@n51, False=LinearInterpolate(A=@n51, B=@n52, Alpha=@n53), Value=NamedRerouteUsage())) {name=UV_withParallax_A, nodeColor=<rgba 1 0.6562 0 1>}
  n54 = NamedRerouteUsage()
  n55 = AppendVector(
   A=ScalarParameter() {group=Styx, parameterName=StyxNoise_Speed2X}
   B=ScalarParameter() {defaultValue=-0.1, group=Styx, parameterName=StyxNoise_Speed2Y}
  )
  n56 = TextureSample(UVs=Add(A=Panner(Coordinate=@n54, Speed=@n55), B=NamedRerouteUsage()), Tex=NamedRerouteUsage()) {samplerSource=SSM_Wrap_WorldGroupSettings}
  n57 = Multiply(
   A=VectorParameter() {group=Styx, parameterName=StyxEmissive2, defaultValue=<rgba 0 5 0.4457 1>}.RGB
   B=Power(Base=@n56.R, Exp=ScalarParameter() {defaultValue=5, group=Styx, parameterName=StyxNoise_Contrast2})
  )
  n58 = Multiply(A=VectorParameter() {group=Styx, parameterName=StyxEmissive, defaultValue=<rgba 0 5 0.4457 1>}.RGB, B=@n56.R)
  n59 = Saturate(None=Power(Base=@n56.R, Exp=ScalarParameter() {defaultValue=5, group=Styx, parameterName=StyxNoise_Contrast}))
  NamedRerouteDeclaration(None=Multiply(A=LinearInterpolate(A=@n57, B=@n58, Alpha=@n59), B=ScalarParameter() {defaultValue=1, group=Styx, parameterName=MainNoiseIntensity})) {name=MainNoisePattern, nodeColor=<rgba 0 1 0.3203 1>}
  n60 = ScalarParameter() {group=Styx_Detail, parameterName=Styx_Detail_Noise_Low}
  n61 = ScalarParameter() {defaultValue=1, group=Styx_Detail, parameterName=Styx_Detail_Noise_High}
  NamedRerouteDeclaration(None=Saturate(None=LinearInterpolate(A=@n60, B=@n61, Alpha=@n32.R))) {name=DetailNoise, nodeColor=<rgba 1 0.0938 0 1>}
  n62 = FunctionInput() {bUsePreviewValueAsDefault=true, inputName=PreSkinnedLocation, sortPriority=2, previewValue=<1 0 0 1>}
  NamedRerouteDeclaration(None=@n62) {name=PreSkinned Loc, nodeColor=<rgba 0.5938 0 1 1>}
  NamedRerouteDeclaration(None=Saturate(None=@n32.G)) {name=MarcoNoiseMask, nodeColor=<rgba 0 1 0.8359 1>}
  n63 = FunctionInput() {bUsePreviewValueAsDefault=true, inputName=Tech Art Mask, sortPriority=1, previewValue=<1 0 0 1>}
  NamedRerouteDeclaration(None=ChannelMaskParameter(None=@n63) {group=Styx, parameterName=Styx Channel}.R) {name=Tech Art Mask, nodeColor=<rgba 1 0.6563 0 1>}
  n64 = NamedRerouteUsage()
  n65 = VectorParameter() {group=StyxMod, parameterName=HeartCenter, defaultValue=<rgba 0 0 128 1>}
  n66 = StaticComponentMaskParameter(None=Subtract(A=@n64, B=ComponentMask(None=@n65.RGB) {b=true, g=true, r=true})) {defaultB=true, defaultR=true, group=StyxMod, parameterName=HeartUVChannel}
  n67 = MaterialFunctionCall!(Coordinate=MaterialFunctionCall!(UV=@n66, Offset, Rotation=ScalarParameter() {group=StyxMod, parameterName=PolarRotate}, Scale, Pivot=Constant()) {materialFunction=/Game/Global/MaterialFunctions/MF_UV_Transform}.Result) {materialFunction=&/PolarCoordinates}
  n68 = ComponentMask(None=@n67.Result) {r=true}
  n69 = ScalarParameter() {defaultValue=3, group=StyxMod, parameterName=PolarXScale}
  n70 = ComponentMask(None=@n67.Result) {g=true}
  n71 = ScalarParameter() {defaultValue=1, group=StyxMod, parameterName=PolarYScale}
  n72 = Divide(
   A=AppendVector(A=Multiply(A=@n68, B=@n69), B=Multiply(A=@n70, B=@n71))
   B=ScalarParameter() {defaultValue=15, group=StyxMod, parameterName=HeartUVSize}
  ) {constB=100}
  n73 = ComponentMask(None=@n72) {g=true}
  NamedRerouteDeclaration(None=Multiply(A=Step(Y=@n73), B=NamedRerouteUsage())) {name=PolarCoordMask, nodeColor=<rgba 0.8281 1 0 1>}
  n74 = NamedRerouteUsage()
  n75 = OneMinus(None=Saturate(None=@n74))
  n76 = Divide(
   A=Subtract(
    A=StaticSwitchParameter(True=@n75, False=@n74) {defaultValue=true, parameterName=RevertDepth}
    B=ScalarParameter() {defaultValue=0.1, group=Styx_Edge, parameterName=Styx_EdgeGlow_Offset}
   )
   B=ScalarParameter() {defaultValue=0.1, group=Styx_Edge, parameterName=Styx_EdgeGlow_Depth}
  )
  NamedRerouteDeclaration(None=OneMinus(None=Saturate(None=@n76))) {name=FakeDepthRevertMask, nodeColor=<rgba 1 0.7266 0 1>}
  n77 = NamedRerouteUsage()
  n78 = Power(Base=NamedRerouteUsage(), Exp=ScalarParameter() {defaultValue=3, group=Styx_Edge, parameterName=Styx_EdgeGlow_Falloff})
  n79 = Power(Base=NamedRerouteUsage(), Exp=ScalarParameter() {defaultValue=3, group=Styx_Edge, parameterName=Styx_EdgeGlow_FalloffFade})
  n80 = MaterialFunctionCall(
   Transition Phase=OneMinus(None=Saturate(None=@n79))
   Height Texture=NamedRerouteUsage()
   Contrast=ScalarParameter() {defaultValue=3, group=Styx_Edge, parameterName=Styx_EdgeGlow_FalloffFadeContrast}
  ) {materialFunction=&HeightLerp}
  n81 = StaticSwitchParameter(True=Multiply(A=@n77, B=Multiply(A=@n78, B=@n80.Alpha)), False=@n77) {group=Styx, parameterName=UseEdgeGlow}
  NamedRerouteDeclaration(None=Multiply(
    A=VectorParameter() {group=Styx_Edge, parameterName=StyxEmissive_EdgeGlow, defaultValue=<rgba 0 5 0.4457 1>}.RGB
    B=NamedRerouteDeclaration(None=@n81) {name=EdgeMask, nodeColor=<rgba 0.2422 0 1 1>}
   )) {name=Final_EdgeGlow, nodeColor=<rgba 0 1 0.0156 1>}
  n82 = MaterialFunctionCall(Float2=@n72) {materialFunction=&BreakOutFloat2Components}
  n83 = TextureSampleParameter2D(UVs=Panner(Coordinate=@n72) {speedY=0.05}) {
   group=StyxMod
   parameterName=PolarDistortUV
   samplerType=SAMPLERTYPE_Grayscale
   texture=/BRCosmetics/Characters/Player/Male/Medium/Bodies/M_MED_DarkStance_Pater/Textures/T_FractalDistortMask
  }
  n84 = Multiply(A=ScalarParameter() {defaultValue=1.5, group=StyxMod, parameterName=PolarDistortIntensity}, B=@n83.R)
  NamedRerouteDeclaration(None=MaterialFunctionCall(X=@n82.R, Y=Add(A=@n84, B=@n82.G)) {materialFunction=&MakeFloat2}.Result) {name=PolarCoordHeart, nodeColor=<rgba 1 0.1641 0 1>}
  n85 = MaterialFunctionCall(Float2=@n66) {materialFunction=&BreakOutFloat2Components}
  n86 = Multiply(A=@n85.G) {constB=-1}
  NamedRerouteDeclaration(None=Divide(
    A=MaterialFunctionCall(X=@n85.R, Y=@n86) {materialFunction=&MakeFloat2}.Result
    B=ScalarParameter() {defaultValue=40, group=StyxMod, parameterName=ShoulderUVSize}
   )) {name=ShoulderCoord, nodeColor=<rgba 0 1 0.1094 1>}
  n87 = TextureSampleParameter2D() {
   group=StyxMod
   parameterName=SpecialUVMask
   samplerType=SAMPLERTYPE_Masks
   texture=/BRCosmetics/Characters/Player/Male/Medium/Bodies/M_MED_DarkStance_Pater/Textures/T_SpecialUVMask
  }
  NamedRerouteDeclaration(None=@n87.R) {name=ShoulderArea, nodeColor=<rgba 0.0312 0 1 1>}
  NamedRerouteDeclaration(None=@n87.G) {name=HeartArea, nodeColor=<rgba 1 0.6797 0 1>}
  n88 = MaterialFunctionCall(Float2=Multiply(A=TextureCoordinate(), B=ScalarParameter() {defaultValue=4, group=Styx, parameterName=StyxMainNoiseTilling_N})) {materialFunction=&BreakOutFloat2Components}
  n89 = Multiply(A=NamedRerouteUsage(), B=@n39) {constB=4}
  n90 = NamedRerouteDeclaration(None=MaterialFunctionCall(X=@n88.R, Y=Add(A=@n88.G, B=@n89)) {materialFunction=&MakeFloat2}.Result) {name=ModUV_N, nodeColor=<rgba 0 1 0.9297 1>}
  n91 = NamedRerouteUsage()
  n92 = NamedRerouteUsage()
  n93 = BumpOffset(
   Coordinate=StaticSwitch(True=@n90, False=LinearInterpolate(A=@n90, B=@n91, Alpha=@n92), Value=NamedRerouteUsage())
   Height=ScalarParameter() {defaultValue=-3, group=Styx, parameterName=StyxUVOffset01}
  )
  n94 = NamedRerouteUsage()
  n95 = NamedRerouteUsage()
  NamedRerouteDeclaration(None=StaticSwitch(True=@n93, False=LinearInterpolate(A=@n93, B=@n94, Alpha=@n95), Value=NamedRerouteUsage())) {name=UV_withParallax_N, nodeColor=<rgba 1 0.6562 0 1>}
  n96 = TextureSampleParameter2D(UVs=@n31) {
   group=StyxMod
   parameterName=DistortSmoke
   samplerType=SAMPLERTYPE_Grayscale
   texture=&T_FractalDistortMask
  }
  n97 = Multiply(A=@n96.RGB, B=ScalarParameter() {defaultValue=10, group=StyxMod, parameterName=SmokeAreaMulti}) {constB=5}
  NamedRerouteDeclaration(None=Multiply(A=Saturate(None=@n97), B=ScalarParameter() {defaultValue=0.5, group=StyxMod, parameterName=SmokeAreaIntensity})) {name=SmokeArea, nodeColor=<rgba 0 1 0.1797 1>}
  n98 = ComponentMask(None=@n64) {b=true}
  n99 = ScalarParameter() {defaultValue=40, group=StyxMod, parameterName=UVRescaleSize}
  n100 = OneMinus(None=Saturate(None=Divide(A=@n98, B=@n99)))
  NamedRerouteDeclaration(None=Power(Base=Saturate(None=@n100))) {name=UVRescaleMask, nodeColor=<rgba 0 0.6953 1 1>}
  NamedRerouteDeclaration(None=@n24) {name=TechArtMaskOverride, nodeColor=<rgba 0.2656 0 1 1>}
  NamedRerouteDeclaration(None=SmoothStep(
    Min=ScalarParameter() {group=StyxMod, parameterName=TechArtMaskMinSpec}
    Max=ScalarParameter() {defaultValue=0.2, group=StyxMod, parameterName=TechArtMaskMaxSpec}
    Value=NamedRerouteUsage()
   )) {name=SpecOverride, nodeColor=<rgba 1 0.7734 0 1>}
  NamedRerouteDeclaration(None=FunctionInput(Preview=StaticBool()) {bUsePreviewValueAsDefault=true, inputName=IsEdge, inputType=FunctionInput_StaticBool, sortPriority=3}) {name=IsEdge, nodeColor=<rgba 0.0781 0 1 1>}
  NamedRerouteDeclaration(None=ComponentMask(None=@n30.RGB) {g=true, r=true}) {name=Distortion_NormalNoise_B, nodeColor=<rgba 0.1016 0 1 1>}
  n101 = FunctionInput() {bUsePreviewValueAsDefault=true, inputName=Material, inputType=FunctionInput_MaterialAttributes}
  NamedRerouteDeclaration(None=GetMaterialAttributes(None=NamedRerouteDeclaration(None=@n101) {name=MA, nodeColor=<rgba 1 0 0.9609 1>}) {attributeGetTypes [#3, #4]}.Normal) {name=Normal IN, nodeColor=<rgba 1 0 0.2578 1>}
  n102 = NamedRerouteUsage()
  n103 = ScalarParameter() {defaultValue=0.01, group=Styx, parameterName=Styx_Distort_UV}
  n104 = Constant()
  NamedRerouteDeclaration(None=Add(A=Multiply(A=@n102, B=@n103), B=@n104)) {name=Distortion_Final_A, nodeColor=<rgba 0.2422 0 1 1>}
  n105 = NamedRerouteUsage()
  NamedRerouteDeclaration(None=Add(A=@n104, B=Multiply(A=@n103, B=@n105))) {name=Distortion_Final_B, nodeColor=<rgba 1 0 0.8906 1>}
  n106 = ScalarParameter() {group=Styx_Edge, parameterName=EdgeGlow_Noise_Contrast_Low}
  n107 = ScalarParameter() {defaultValue=1, group=Styx_Edge, parameterName=EdgeGlow_Noise_Contrast_HIGH}
  n108 = NamedRerouteUsage()
  n109 = ScalarParameter() {defaultValue=0.1, group=Styx_Edge, parameterName=Styx_NoiseEdge_Distortion}
  n110 = TextureSample(UVs=Add(A=NamedRerouteUsage(), B=Multiply(A=@n108, B=@n109)), Tex=NamedRerouteUsage())
  NamedRerouteDeclaration(None=Saturate(None=LinearInterpolate(A=@n106, B=@n107, Alpha=@n110.G))) {name=NoisePerlin, nodeColor=<rgba 0.2188 0 1 1>}
  n111 = TextureObjectParameter() {
   group=Styx
   parameterName=Styx Mask
   samplerType=SAMPLERTYPE_Masks
   texture=/Game/Characters/Player/Male/Medium/Bodies/M_MED_DarkStance_Pater/Textures/T_StyxCosmetic_Mask
  }
  NamedRerouteDeclaration(None=@n111) {name=Styx Mask IN, nodeColor=<rgba 1 0.8906 0 1>}
  NamedRerouteUsage()
  NamedRerouteUsage()
  NamedRerouteUsage()
  NamedRerouteUsage()
  n112 = Sine(None=Multiply(A=@n17, B=ScalarParameter() {defaultValue=4, group=StyxMod, parameterName=FakeMetalicSinMulti}) {constB=2})
  StaticSwitchParameter(True=Multiply(A=ConstantBiasScale(None=@n112), B=@n17), False=@n17) {group=StyxMod, parameterName=TyrianVersion}
 ]}
