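# Character body material for the DarkStance Pater cosmetic that gives edges a dissolving, smoky look using dithered masked opacity despite its translucent name.
# It runs the 2023 character parent PreFX, then MF_StyxCosmetic in edge mode, then scales emissive and overrides opacity mask before PostFX.
# Opacity combines pre-skinned position gradients from BottomLoc with a smoke-area mask and panning noise layers, both UV-distorted, then remaps the result with smoothstep.
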Material /BRCosmetics/Characters/Player/Male/Medium/Bodies/M_MED_DarkStance_Pater/Materials/M_TranSlucentEdges
graph {
 properties {
  allowTranslucentCustomDepthWrites=true
  bUseMaterialAttributes=true
  bUseTranslucencyVertexFog=false
  bUsedWithMeshDeformer=true
  bUsedWithMorphTargets=true
  bUsedWithSkeletalMesh=true
  blendMode=BLEND_Masked
  ditherOpacityMask=true
  opacityMaskClipValue=0.5
  pixelDepthOffsetMode=PDOM_Legacy
  shadingModel=MSM_FromMaterialExpression
  translucencyLightingMode=TLM_VolumetricPerVertexDirectional
  usedShadingModels=Default Lit | Subsurface | Clear Coat | Subsurface Profile | Cloth
 }
 nodes [
  Constant()
  n1 = MaterialFunctionCall() {materialFunction=&MF_CharacterParent_2023_PreFX}
  n2 = MaterialFunctionCall!(
   Material=@n1.Result
   Tech Art Mask=@n1.Result
   PreSkinnedLocation=NamedRerouteDeclaration(None=@n1.Pre skinned Local Position) {name=PreSkinnedLoc, nodeColor=<rgba 1 0.1875 0 1>}
   IsEdge=StaticBool() {value=true}
  ) {materialFunction=&/MF_StyxCosmetic}
  n3 = GetMaterialAttributes(None=@n2.Result) {attributeGetTypes [#1]}
  n4 = ScalarParameter() {defaultValue=1, group=OpacityControl, parameterName=EmissivePostMulti}
  MaterialFunctionCall(
   Material=SetMaterialAttributes!(MaterialAttributes=@n3.MaterialAttributes, Emissive Color=Multiply(A=@n3.Emissive Color, B=@n4), Opacity Mask=NamedRerouteUsage()) {attributeSetTypes [#1, #2]}
   M=@n1.M
   Pre Skinned Local Position=@n1.M
   Distance Blend=@n1.M
   Fresnel=@n1.M
  ) {materialFunction=&MF_CharacterParent_2023_PostFX}
  n5 = Subtract(A=NamedRerouteUsage(), B=VectorParameter() {group=OpacityControl, parameterName=BottomLoc, defaultValue=<rgba 0 0 10 1>}.RGB)
  n6 = ComponentMask(None=@n5) {b=true}
  n7 = ScalarParameter() {defaultValue=80, group=OpacityControl, parameterName=OpacityGradientUnit01}
  n8 = Saturate(None=Divide(A=@n6, B=@n7) {constB=50})
  n9 = NamedRerouteDeclaration(None=@n8) {name=OpacityGradient01, nodeColor=<rgba 1 0 0.2812 1>}
  n10 = ComponentMask(None=@n5) {g=true}
  n11 = ScalarParameter() {defaultValue=10, group=OpacityControl, parameterName=OpacityGradientUnit05}
  n12 = Saturate(None=Divide(A=@n6, B=Add(A=@n7, B=@n11) {constB=5}))
  n13 = Multiply(A=Saturate(None=Divide(A=@n10, B=@n11) {constB=5}), B=Multiply(A=OneMinus(None=@n12), B=@n8))
  n14 = SmoothStep(
   Min=ScalarParameter() {defaultValue=0.4, group=OpacityControl, parameterName=GradientMin01}
   Max=ScalarParameter() {defaultValue=0.7, group=OpacityControl, parameterName=GradientMax01}
   Value=Saturate(None=Subtract(A=@n9, B=@n13))
  )
  n15 = NamedRerouteUsage()
  n16 = ComponentMask(None=Multiply(A=@n15, B=ScalarParameter() {defaultValue=0.02, group=OpacityControl, parameterName=Distort_Intensity02}) {constB=0.2}) {r=true}
  n17 = ComponentMask(None=Multiply(A=@n15, B=ScalarParameter() {group=OpacityControl, parameterName=Distort_Intensity03})) {g=true}
  n18 = TextureSampleParameter2D(UVs=Add(A=TextureCoordinate(), B=AppendVector(A=@n16, B=@n17))) {
   group=OpacityControl
   parameterName=SmokeArea
   samplerType=SAMPLERTYPE_Masks
   texture=/BRCosmetics/Characters/Player/Male/Medium/Bodies/M_MED_DarkStance_Pater/Textures/T_DarkStanceSmokeArea
  }
  n19 = Multiply(A=@n18.R)
  n20 = Multiply(A=TextureCoordinate(), B=ScalarParameter() {defaultValue=10, group=OpacityControl, parameterName=NoiseTilling}) {constB=30}
  n21 = Multiply(A=NamedRerouteUsage(), B=ScalarParameter() {defaultValue=0.1, group=OpacityControl, parameterName=Distort_Intensity01}) {constB=0.2}
  n22 = ScalarParameter() {group=OpacityControl, parameterName=OpacityNoiseSpeedX}
  n23 = ScalarParameter() {defaultValue=-0.2, group=OpacityControl, parameterName=OpacityNoiseSpeedY}
  n24 = TextureSampleParameter2D(UVs=Panner(Coordinate=Add(A=@n20, B=@n21), Speed=AppendVector(A=@n22, B=@n23))) {
   group=OpacityControl
   parameterName=OpacityNoise
   samplerType=SAMPLERTYPE_Masks
   texture=/BRCosmetics/Characters/Player/Male/Medium/Bodies/M_MED_DarkStance_Pater/Textures/T_DarkStanceNoiseMask
  }
  n25 = ComponentMask(None=@n24.RGB) {b=true, g=true, r=true}
  n26 = NamedRerouteUsage()
  n27 = Saturate(None=Subtract(A=@n25, B=OneMinus(None=@n26)))
  n28 = ComponentMask(None=@n27) {r=true}
  n29 = Multiply(A=ComponentMask(None=@n27) {g=true}, B=ScalarParameter() {defaultValue=1, group=OpacityControl, parameterName=SecondSmokeIntensity}) {constB=0.5}
  n30 = Subtract(A=ComponentMask(None=@n25) {b=true}, B=OneMinus(None=NamedRerouteUsage()))
  n31 = Saturate(None=Add(A=Add(A=@n28, B=@n29), B=Saturate(None=@n30)))
  n32 = NamedRerouteUsage()
  n33 = Power(Base=@n32)
  n34 = Multiply(A=@n18.G, B=Add(A=Saturate(None=@n19), B=Multiply(A=@n31, B=@n33)))
  n35 = Multiply(A=Saturate(None=Add(A=@n14, B=@n34)), B=ScalarParameter() {defaultValue=2, group=OpacityControl, parameterName=OverallOpacityScale})
  NamedRerouteDeclaration(None=SmoothStep(
    Min=ScalarParameter() {group=OpacityControl, parameterName=OverallOpacityMin}
    Max=ScalarParameter() {defaultValue=0.8, group=OpacityControl, parameterName=OverallOpacityMax}
    Value=Saturate(None=@n35)
   )) {name=Opacity_OUT, nodeColor=<rgba 0 0.1094 1 1>}
  n36 = ScalarParameter() {defaultValue=40, group=OpacityControl, parameterName=OpacityGradientUnit02}
  NamedRerouteDeclaration(None=Saturate(None=Divide(A=@n6, B=@n36) {constB=50})) {name=OpacityGradient02, nodeColor=<rgba 1 0 0.2812 1>}
  NamedRerouteDeclaration(None=@n2.UV_Distort) {name=UVDistort, nodeColor=<rgba 0.8281 0 1 1>}
  n37 = ScalarParameter() {defaultValue=3, group=OpacityControl, parameterName=OpacityGradientUnit03}
  NamedRerouteDeclaration(None=Saturate(None=Divide(A=@n6, B=@n37) {constB=50})) {name=OpacityGradient03, nodeColor=<rgba 1 0 0.2812 1>}
  n38 = ScalarParameter() {defaultValue=10, group=OpacityControl, parameterName=OpacityGradientUnit04}
  NamedRerouteDeclaration(None=Saturate(None=Divide(A=@n6, B=@n38) {constB=50})) {name=OpacityGradient04, nodeColor=<rgba 1 0 0.2812 1>}
  OneMinus(None=@n32)
 ]
}
