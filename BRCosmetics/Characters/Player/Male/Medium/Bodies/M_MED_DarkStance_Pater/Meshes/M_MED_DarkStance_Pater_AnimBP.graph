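# Cosmetic body AnimBlueprint for M_MED_DarkStance_Pater that animates a long coat over the pose copied from the attached parent mesh.
# A Control Rig feeds AnimDynamics on five five-bone coat chains, then additive bone offsets, constraints blending neighboring panels, and finger-scale bone adjustments.
# The event graph derives simulation space, movement, vehicle and gravity values from DBNO, front-end and movement state, lowering MovementAlpha for open DarkStance-tagged gliders.
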AnimBlueprint /BRCosmetics/Characters/Player/Male/Medium/Bodies/M_MED_DarkStance_Pater/Meshes/M_MED_DarkStance_Pater_AnimBP
target_skeleton=&/M_MED_DarkStance_Pater_Skeleton
graphs [
 AnimGraph {nodes [
   n1 = Control Rig!(
    Source=Copy Pose From Mesh() {bUseAttachedParent=true}
    Alpha
    MovementAlpha:double=Get MovementAlpha() -> (MovementAlpha:double)
    is_DBNO_Carried:bool=Get bIsDBNOCarried()
   ) -> (Pose) {
    alphaBoolBlend {blendOption=Linear}
    controlRigAssetReference {blueprintRigClass=&/M_MED_DarkStance_Pater_CtrlRig.M_MED_DarkStance_Pater_CtrlRig_C}
   }
   n2 = Anim Dynamics - Bone: dyn_L_front_coat_1!(
    SimulationSpace:AnimPhysSimSpaceType Enum
    ComponentPose=Local To Component(LocalPose=@n1)
    Alpha=1
    GravityOverride:Vector="0,0,-980"
    ExternalForce="0,0,0"
   ) -> (Pose) {
    angularDampingOverride=0.99
    angularSpringConstant=50
    bAngularSpring=true
    bOverrideAngularDamping=true
    bUseGravityOverride=true
    gravityScale=0.35
    simulationSpace=RootRelative
    alphaBoolBlend {blendOption=Linear}
    boundBone {boneName=dyn_L_front_coat_1}
    componentLinearAccScale=<0.4 0.4 0.4>
    componentLinearVelScale=<0.4 0.4 0.4>
    gravityOverride {z=-980}
    physicsBodyDefinitions [{
      boundBone {boneName=dyn_L_front_coat_1}
      boxExtents=<14 14 14>
      constraintSetup {angularLimitsMax=<10 20 0>, angularLimitsMin=<-10 -6 -15>, angularTarget=<1 0 0>}
      localJointOffset=<2 0 0>
     }]
   }
   n3 = Anim Dynamics - Bone: dyn_L_front_coat_2!(SimulationSpace:AnimPhysSimSpaceType Enum, ComponentPose=@n2, Alpha=1, GravityOverride:Vector="0,0,-1000", ExternalForce="0,0,0") -> (Pose) {
    angularDampingOverride=0.999
    angularSpringConstant=50
    bAngularSpring=true
    bOverrideAngularDamping=true
    bUseGravityOverride=true
    gravityScale=0.35
    simulationSpace=RootRelative
    alphaBoolBlend {blendOption=Linear}
    boundBone {boneName=dyn_L_front_coat_2}
    componentLinearAccScale=<0.4 0.4 0.4>
    componentLinearVelScale=<0.4 0.4 0.4>
    gravityOverride {z=-1000}
    physicsBodyDefinitions [{
      boundBone {boneName=dyn_L_front_coat_2}
      boxExtents=<14 14 14>
      constraintSetup {angularLimitsMax=<10 20 0>, angularLimitsMin=<-10 -6 -15>, angularTarget=<1 0 0>}
      localJointOffset=<2 0 0>
     }]
   }
   n4 = Anim Dynamics - Bone: dyn_L_front_coat_3!(SimulationSpace:AnimPhysSimSpaceType Enum, ComponentPose=@n3, Alpha=1, GravityOverride:Vector="0,0,-1000", ExternalForce="0,0,0") -> (Pose) {
    angularDampingOverride=0.999
    angularSpringConstant=50
    bAngularSpring=true
    bOverrideAngularDamping=true
    bUseGravityOverride=true
    gravityScale=0.35
    simulationSpace=RootRelative
    alphaBoolBlend {blendOption=Linear}
    boundBone {boneName=dyn_L_front_coat_3}
    componentLinearAccScale=<0.4 0.4 0.4>
    componentLinearVelScale=<0.4 0.4 0.4>
    gravityOverride {z=-1000}
    physicsBodyDefinitions [{
      boundBone {boneName=dyn_L_front_coat_3}
      boxExtents=<14 14 14>
      constraintSetup {angularLimitsMax=<10 20 0>, angularLimitsMin=<-10 -6 -15>, angularTarget=<1 0 0>}
      localJointOffset=<2 0 0>
     }]
   }
   n5 = Anim Dynamics - Bone: dyn_L_front_coat_4!(SimulationSpace:AnimPhysSimSpaceType Enum, ComponentPose=@n4, Alpha=1, GravityOverride:Vector="0,0,-1000", ExternalForce="0,0,0") -> (Pose) {
    angularDampingOverride=0.999
    angularSpringConstant=50
    bAngularSpring=true
    bOverrideAngularDamping=true
    bUseGravityOverride=true
    gravityScale=0.35
    simulationSpace=RootRelative
    alphaBoolBlend {blendOption=Linear}
    boundBone {boneName=dyn_L_front_coat_4}
    componentLinearAccScale=<0.4 0.4 0.4>
    componentLinearVelScale=<0.4 0.4 0.4>
    gravityOverride {z=-1000}
    physicsBodyDefinitions [{
      boundBone {boneName=dyn_L_front_coat_4}
      boxExtents=<14 14 14>
      constraintSetup {angularLimitsMax=<10 20 0>, angularLimitsMin=<-10 -6 -15>, angularTarget=<1 0 0>}
      localJointOffset=<2 0 0>
     }]
   }
   n6 = Anim Dynamics - Bone: dyn_L_front_coat_5!(SimulationSpace:AnimPhysSimSpaceType Enum, ComponentPose=@n5, Alpha=1, GravityOverride:Vector="0,0,-1000", ExternalForce="0,0,0") -> (Pose) {
    angularDampingOverride=0.999
    angularSpringConstant=50
    bAngularSpring=true
    bOverrideAngularDamping=true
    bUseGravityOverride=true
    gravityScale=0.35
    simulationSpace=RootRelative
    alphaBoolBlend {blendOption=Linear}
    boundBone {boneName=dyn_L_front_coat_5}
    componentLinearAccScale=<0.4 0.4 0.4>
    componentLinearVelScale=<0.4 0.4 0.4>
    gravityOverride {z=-1000}
    physicsBodyDefinitions [{
      boundBone {boneName=dyn_L_front_coat_5}
      boxExtents=<14 14 14>
      constraintSetup {angularLimitsMax=<10 20 0>, angularLimitsMin=<-10 -6 -15>, angularTarget=<1 0 0>}
      localJointOffset=<2 0 0>
     }]
   }
   n7 = Anim Dynamics - Bone: dyn_L_side_coat_1!(SimulationSpace:AnimPhysSimSpaceType Enum, ComponentPose=@n6, Alpha=1, GravityOverride:Vector="0,0,-980", ExternalForce="0,0,0") -> (Pose) {
    angularDampingOverride=0.999
    angularSpringConstant=50
    bAngularSpring=true
    bUseGravityOverride=true
    gravityScale=0.35
    simulationSpace=RootRelative
    alphaBoolBlend {blendOption=Linear}
    boundBone {boneName=dyn_L_side_coat_1}
    componentLinearAccScale=<0.4 0.4 0.4>
    componentLinearVelScale=<0.4 0.4 0.4>
    gravityOverride {z=-980}
    physicsBodyDefinitions [{
      boundBone {boneName=dyn_L_side_coat_1}
      boxExtents=<14 14 14>
      constraintSetup {angularLimitsMax=<10 20 0>, angularLimitsMin=<-10 -6 -15>, angularTarget=<1 0 0>}
      localJointOffset=<2 0 0>
     }]
   }
   n8 = Anim Dynamics - Bone: dyn_L_side_coat_2!(SimulationSpace:AnimPhysSimSpaceType Enum, ComponentPose=@n7, Alpha=1, GravityOverride:Vector="0,0,-1000", ExternalForce="0,0,0") -> (Pose) {
    angularDampingOverride=0.999
    angularSpringConstant=50
    bAngularSpring=true
    bOverrideAngularDamping=true
    bUseGravityOverride=true
    gravityScale=0.35
    simulationSpace=RootRelative
    alphaBoolBlend {blendOption=Linear}
    boundBone {boneName=dyn_L_side_coat_2}
    componentLinearAccScale=<0.4 0.4 0.4>
    componentLinearVelScale=<0.4 0.4 0.4>
    gravityOverride {z=-1000}
    physicsBodyDefinitions [{
      boundBone {boneName=dyn_L_side_coat_2}
      boxExtents=<14 14 14>
      constraintSetup {angularLimitsMax=<10 20 0>, angularLimitsMin=<-10 -6 -15>, angularTarget=<1 0 0>}
      localJointOffset=<2 0 0>
     }]
   }
   n9 = Anim Dynamics - Bone: dyn_L_side_coat_3!(SimulationSpace:AnimPhysSimSpaceType Enum, ComponentPose=@n8, Alpha=1, GravityOverride:Vector="0,0,-1000", ExternalForce="0,0,0") -> (Pose) {
    angularDampingOverride=0.999
    angularSpringConstant=50
    bAngularSpring=true
    bOverrideAngularDamping=true
    bUseGravityOverride=true
    gravityScale=0.35
    simulationSpace=RootRelative
    alphaBoolBlend {blendOption=Linear}
    boundBone {boneName=dyn_L_side_coat_3}
    componentLinearAccScale=<0.4 0.4 0.4>
    componentLinearVelScale=<0.4 0.4 0.4>
    gravityOverride {z=-1000}
    physicsBodyDefinitions [{
      boundBone {boneName=dyn_L_side_coat_3}
      boxExtents=<14 14 14>
      constraintSetup {angularLimitsMax=<10 20 0>, angularLimitsMin=<-10 -6 -15>, angularTarget=<1 0 0>}
      localJointOffset=<2 0 0>
     }]
   }
   n10 = Anim Dynamics - Bone: dyn_L_side_coat_4!(SimulationSpace:AnimPhysSimSpaceType Enum, ComponentPose=@n9, Alpha=1, GravityOverride:Vector="0,0,-1000", ExternalForce="0,0,0") -> (Pose) {
    angularDampingOverride=0.999
    angularSpringConstant=50
    bAngularSpring=true
    bOverrideAngularDamping=true
    bUseGravityOverride=true
    gravityScale=0.35
    simulationSpace=RootRelative
    alphaBoolBlend {blendOption=Linear}
    boundBone {boneName=dyn_L_side_coat_4}
    componentLinearAccScale=<0.4 0.4 0.4>
    componentLinearVelScale=<0.4 0.4 0.4>
    gravityOverride {z=-1000}
    physicsBodyDefinitions [{
      boundBone {boneName=dyn_L_side_coat_4}
      boxExtents=<14 14 14>
      constraintSetup {angularLimitsMax=<10 20 0>, angularLimitsMin=<-10 -6 -15>, angularTarget=<1 0 0>}
      localJointOffset=<2 0 0>
     }]
   }
   n11 = Anim Dynamics - Bone: dyn_L_side_coat_5!(SimulationSpace:AnimPhysSimSpaceType Enum, ComponentPose=@n10, Alpha=1, GravityOverride:Vector="0,0,-1000", ExternalForce="0,0,0") -> (Pose) {
    angularDampingOverride=0.999
    angularSpringConstant=50
    bAngularSpring=true
    bOverrideAngularDamping=true
    bUseGravityOverride=true
    gravityScale=0.35
    simulationSpace=RootRelative
    alphaBoolBlend {blendOption=Linear}
    boundBone {boneName=dyn_L_side_coat_5}
    componentLinearAccScale=<0.4 0.4 0.4>
    componentLinearVelScale=<0.4 0.4 0.4>
    gravityOverride {z=-1000}
    physicsBodyDefinitions [{
      boundBone {boneName=dyn_L_side_coat_5}
      boxExtents=<14 14 14>
      constraintSetup {angularLimitsMax=<10 20 0>, angularLimitsMin=<-10 -6 -15>, angularTarget=<1 0 0>}
      localJointOffset=<2 0 0>
     }]
   }
   n12 = Anim Dynamics - Bone: dyn_R_front_coat_1!(SimulationSpace:AnimPhysSimSpaceType Enum, ComponentPose=@n11, Alpha=1, GravityOverride:Vector="0,0,-980", ExternalForce="0,0,0") -> (Pose) {
    angularDampingOverride=0.999
    angularSpringConstant=50
    bAngularSpring=true
    bUseGravityOverride=true
    gravityScale=0.35
    simulationSpace=RootRelative
    alphaBoolBlend {blendOption=Linear}
    boundBone {boneName=dyn_R_front_coat_1}
    componentLinearAccScale=<0.4 0.4 0.4>
    componentLinearVelScale=<0.4 0.4 0.4>
    gravityOverride {z=-980}
    physicsBodyDefinitions [{
      boundBone {boneName=dyn_R_front_coat_1}
      boxExtents=<14 14 14>
      constraintSetup {angularLimitsMax=<10 20 0>, angularLimitsMin=<-10 -6 -15>, angularTarget=<1 0 0>}
      localJointOffset=<-2 0 0>
     }]
   }
   n13 = Anim Dynamics - Bone: dyn_R_front_coat_2!(SimulationSpace:AnimPhysSimSpaceType Enum, ComponentPose=@n12, Alpha=1, GravityOverride:Vector="0,0,-1000", ExternalForce="0,0,0") -> (Pose) {
    angularDampingOverride=0.999
    angularSpringConstant=50
    bAngularSpring=true
    bOverrideAngularDamping=true
    bUseGravityOverride=true
    gravityScale=0.35
    simulationSpace=RootRelative
    alphaBoolBlend {blendOption=Linear}
    boundBone {boneName=dyn_R_front_coat_2}
    componentLinearAccScale=<0.4 0.4 0.4>
    componentLinearVelScale=<0.4 0.4 0.4>
    gravityOverride {z=-1000}
    physicsBodyDefinitions [{
      boundBone {boneName=dyn_R_front_coat_2}
      boxExtents=<14 14 14>
      constraintSetup {angularLimitsMax=<10 20 0>, angularLimitsMin=<-10 -6 -15>, angularTarget=<1 0 0>}
      localJointOffset=<-2 0 0>
     }]
   }
   n14 = Anim Dynamics - Bone: dyn_R_front_coat_3!(SimulationSpace:AnimPhysSimSpaceType Enum, ComponentPose=@n13, Alpha=1, GravityOverride:Vector="0,0,-1000", ExternalForce="0,0,0") -> (Pose) {
    angularDampingOverride=0.999
    angularSpringConstant=50
    bAngularSpring=true
    bOverrideAngularDamping=true
    bUseGravityOverride=true
    gravityScale=0.35
    simulationSpace=RootRelative
    alphaBoolBlend {blendOption=Linear}
    boundBone {boneName=dyn_R_front_coat_3}
    componentLinearAccScale=<0.4 0.4 0.4>
    componentLinearVelScale=<0.4 0.4 0.4>
    gravityOverride {z=-1000}
    physicsBodyDefinitions [{
      boundBone {boneName=dyn_R_front_coat_3}
      boxExtents=<14 14 14>
      constraintSetup {angularLimitsMax=<10 20 0>, angularLimitsMin=<-10 -6 -15>, angularTarget=<1 0 0>}
      localJointOffset=<-2 0 0>
     }]
   }
   n15 = Anim Dynamics - Bone: dyn_R_front_coat_4!(SimulationSpace:AnimPhysSimSpaceType Enum, ComponentPose=@n14, Alpha=1, GravityOverride:Vector="0,0,-1000", ExternalForce="0,0,0") -> (Pose) {
    angularDampingOverride=0.999
    angularSpringConstant=50
    bAngularSpring=true
    bOverrideAngularDamping=true
    bUseGravityOverride=true
    gravityScale=0.35
    simulationSpace=RootRelative
    alphaBoolBlend {blendOption=Linear}
    boundBone {boneName=dyn_R_front_coat_4}
    componentLinearAccScale=<0.4 0.4 0.4>
    componentLinearVelScale=<0.4 0.4 0.4>
    gravityOverride {z=-1000}
    physicsBodyDefinitions [{
      boundBone {boneName=dyn_R_front_coat_4}
      boxExtents=<14 14 14>
      constraintSetup {angularLimitsMax=<10 20 0>, angularLimitsMin=<-10 -6 -15>, angularTarget=<1 0 0>}
      localJointOffset=<-2 0 0>
     }]
   }
   n16 = Anim Dynamics - Bone: dyn_R_front_coat_5!(SimulationSpace:AnimPhysSimSpaceType Enum, ComponentPose=@n15, Alpha=1, GravityOverride:Vector="0,0,-1000", ExternalForce="0,0,0") -> (Pose) {
    angularDampingOverride=0.999
    angularSpringConstant=50
    bAngularSpring=true
    bOverrideAngularDamping=true
    bUseGravityOverride=true
    gravityScale=0.35
    simulationSpace=RootRelative
    alphaBoolBlend {blendOption=Linear}
    boundBone {boneName=dyn_R_front_coat_5}
    componentLinearAccScale=<0.4 0.4 0.4>
    componentLinearVelScale=<0.4 0.4 0.4>
    gravityOverride {z=-1000}
    physicsBodyDefinitions [{
      boundBone {boneName=dyn_R_front_coat_5}
      boxExtents=<14 14 14>
      constraintSetup {angularLimitsMax=<10 20 0>, angularLimitsMin=<-10 -6 -15>, angularTarget=<1 0 0>}
      localJointOffset=<-2 0 0>
     }]
   }
   n17 = Anim Dynamics - Bone: dyn_R_side_coat_1!(SimulationSpace:AnimPhysSimSpaceType Enum, ComponentPose=@n16, Alpha=1, GravityOverride:Vector="0,0,-980", ExternalForce="0,0,0") -> (Pose) {
    angularDampingOverride=0.999
    angularSpringConstant=50
    bAngularSpring=true
    bUseGravityOverride=true
    gravityScale=0.35
    simulationSpace=RootRelative
    alphaBoolBlend {blendOption=Linear}
    boundBone {boneName=dyn_R_side_coat_1}
    componentLinearAccScale=<0.4 0.4 0.4>
    componentLinearVelScale=<0.4 0.4 0.4>
    gravityOverride {z=-980}
    physicsBodyDefinitions [{
      boundBone {boneName=dyn_R_side_coat_1}
      boxExtents=<14 14 14>
      constraintSetup {angularLimitsMax=<10 20 0>, angularLimitsMin=<-10 -6 -15>, angularTarget=<1 0 0>}
      localJointOffset=<-2 0 0>
     }]
   }
   n18 = Anim Dynamics - Bone: dyn_R_side_coat_2!(SimulationSpace:AnimPhysSimSpaceType Enum, ComponentPose=@n17, Alpha=1, GravityOverride:Vector="0,0,-1000", ExternalForce="0,0,0") -> (Pose) {
    angularDampingOverride=0.999
    angularSpringConstant=50
    bAngularSpring=true
    bOverrideAngularDamping=true
    bUseGravityOverride=true
    gravityScale=0.35
    simulationSpace=RootRelative
    alphaBoolBlend {blendOption=Linear}
    boundBone {boneName=dyn_R_side_coat_2}
    componentLinearAccScale=<0.4 0.4 0.4>
    componentLinearVelScale=<0.4 0.4 0.4>
    gravityOverride {z=-1000}
    physicsBodyDefinitions [{
      boundBone {boneName=dyn_R_side_coat_2}
      boxExtents=<14 14 14>
      constraintSetup {angularLimitsMax=<10 20 0>, angularLimitsMin=<-10 -6 -15>, angularTarget=<1 0 0>}
      localJointOffset=<-2 0 0>
     }]
   }
   n19 = Anim Dynamics - Bone: dyn_R_side_coat_3!(SimulationSpace:AnimPhysSimSpaceType Enum, ComponentPose=@n18, Alpha=1, GravityOverride:Vector="0,0,-1000", ExternalForce="0,0,0") -> (Pose) {
    angularDampingOverride=0.999
    angularSpringConstant=50
    bAngularSpring=true
    bOverrideAngularDamping=true
    bUseGravityOverride=true
    gravityScale=0.35
    simulationSpace=RootRelative
    alphaBoolBlend {blendOption=Linear}
    boundBone {boneName=dyn_R_side_coat_3}
    componentLinearAccScale=<0.4 0.4 0.4>
    componentLinearVelScale=<0.4 0.4 0.4>
    gravityOverride {z=-1000}
    physicsBodyDefinitions [{
      boundBone {boneName=dyn_R_side_coat_3}
      boxExtents=<14 14 14>
      constraintSetup {angularLimitsMax=<10 20 0>, angularLimitsMin=<-10 -6 -15>, angularTarget=<1 0 0>}
      localJointOffset=<-2 0 0>
     }]
   }
   n20 = Anim Dynamics - Bone: dyn_R_side_coat_4!(SimulationSpace:AnimPhysSimSpaceType Enum, ComponentPose=@n19, Alpha=1, GravityOverride:Vector="0,0,-1000", ExternalForce="0,0,0") -> (Pose) {
    angularDampingOverride=0.999
    angularSpringConstant=50
    bAngularSpring=true
    bOverrideAngularDamping=true
    bUseGravityOverride=true
    gravityScale=0.35
    simulationSpace=RootRelative
    alphaBoolBlend {blendOption=Linear}
    boundBone {boneName=dyn_R_side_coat_4}
    componentLinearAccScale=<0.4 0.4 0.4>
    componentLinearVelScale=<0.4 0.4 0.4>
    gravityOverride {z=-1000}
    physicsBodyDefinitions [{
      boundBone {boneName=dyn_R_side_coat_4}
      boxExtents=<14 14 14>
      constraintSetup {angularLimitsMax=<10 20 0>, angularLimitsMin=<-10 -6 -15>, angularTarget=<1 0 0>}
      localJointOffset=<-2 0 0>
     }]
   }
   n21 = Anim Dynamics - Bone: dyn_R_side_coat_5!(SimulationSpace:AnimPhysSimSpaceType Enum, ComponentPose=@n20, Alpha=1, GravityOverride:Vector="0,0,-1000", ExternalForce="0,0,0") -> (Pose) {
    angularDampingOverride=0.999
    angularSpringConstant=50
    bAngularSpring=true
    bOverrideAngularDamping=true
    bUseGravityOverride=true
    gravityScale=0.35
    simulationSpace=RootRelative
    alphaBoolBlend {blendOption=Linear}
    boundBone {boneName=dyn_R_side_coat_5}
    componentLinearAccScale=<0.4 0.4 0.4>
    componentLinearVelScale=<0.4 0.4 0.4>
    gravityOverride {z=-1000}
    physicsBodyDefinitions [{
      boundBone {boneName=dyn_R_side_coat_5}
      boxExtents=<14 14 14>
      constraintSetup {angularLimitsMax=<10 20 0>, angularLimitsMin=<-10 -6 -15>, angularTarget=<1 0 0>}
      localJointOffset=<-2 0 0>
     }]
   }
   n22 = Anim Dynamics - Bone: dyn_C_back_coat_1!(SimulationSpace:AnimPhysSimSpaceType Enum, ComponentPose=@n21, Alpha=1, GravityOverride:Vector="0,0,-980", ExternalForce="0,0,0") -> (Pose) {
    angularDampingOverride=0.999
    angularSpringConstant=50
    bAngularSpring=true
    bUseGravityOverride=true
    gravityScale=0.35
    simulationSpace=RootRelative
    alphaBoolBlend {blendOption=Linear}
    boundBone {boneName=dyn_C_back_coat_1}
    componentLinearAccScale=<0.4 0.4 0.4>
    componentLinearVelScale=<0.4 0.4 0.4>
    gravityOverride {z=-980}
    physicsBodyDefinitions [{
      boundBone {boneName=dyn_C_back_coat_1}
      boxExtents=<17 17 17>
      constraintSetup {angularLimitsMax=<10 20 0>, angularLimitsMin=<-10 -20 -15>, angularTarget=<1 0 0>}
      localJointOffset=<2 0 0>
     }]
   }
   n23 = Anim Dynamics - Bone: dyn_C_back_coat_2!(SimulationSpace:AnimPhysSimSpaceType Enum, ComponentPose=@n22, Alpha=1, GravityOverride:Vector="0,0,-1000", ExternalForce="0,0,0") -> (Pose) {
    angularDampingOverride=0.999
    angularSpringConstant=50
    bAngularSpring=true
    bOverrideAngularDamping=true
    bUseGravityOverride=true
    gravityScale=0.35
    simulationSpace=RootRelative
    alphaBoolBlend {blendOption=Linear}
    boundBone {boneName=dyn_C_back_coat_2}
    componentLinearAccScale=<0.4 0.4 0.4>
    componentLinearVelScale=<0.4 0.4 0.4>
    gravityOverride {z=-1000}
    physicsBodyDefinitions [{
      boundBone {boneName=dyn_C_back_coat_2}
      boxExtents=<17 17 17>
      constraintSetup {angularLimitsMax=<10 20 0>, angularLimitsMin=<-10 -20 -15>, angularTarget=<1 0 0>}
      localJointOffset=<2 0 0>
     }]
   }
   n24 = Anim Dynamics - Bone: dyn_C_back_coat_3!(SimulationSpace:AnimPhysSimSpaceType Enum, ComponentPose=@n23, Alpha=1, GravityOverride:Vector="0,0,-1000", ExternalForce="0,0,0") -> (Pose) {
    angularDampingOverride=0.999
    angularSpringConstant=50
    bAngularSpring=true
    bOverrideAngularDamping=true
    bUseGravityOverride=true
    gravityScale=0.35
    simulationSpace=RootRelative
    alphaBoolBlend {blendOption=Linear}
    boundBone {boneName=dyn_C_back_coat_3}
    componentLinearAccScale=<0.4 0.4 0.4>
    componentLinearVelScale=<0.4 0.4 0.4>
    gravityOverride {z=-1000}
    physicsBodyDefinitions [{
      boundBone {boneName=dyn_C_back_coat_3}
      boxExtents=<17 17 17>
      constraintSetup {angularLimitsMax=<10 20 0>, angularLimitsMin=<-10 -20 -15>, angularTarget=<1 0 0>}
      localJointOffset=<2 0 0>
     }]
   }
   n25 = Anim Dynamics - Bone: dyn_C_back_coat_4!(SimulationSpace:AnimPhysSimSpaceType Enum, ComponentPose=@n24, Alpha=1, GravityOverride:Vector="0,0,-1000", ExternalForce="0,0,0") -> (Pose) {
    angularDampingOverride=0.999
    angularSpringConstant=50
    bAngularSpring=true
    bOverrideAngularDamping=true
    bUseGravityOverride=true
    gravityScale=0.35
    simulationSpace=RootRelative
    alphaBoolBlend {blendOption=Linear}
    boundBone {boneName=dyn_C_back_coat_4}
    componentLinearAccScale=<0.4 0.4 0.4>
    componentLinearVelScale=<0.4 0.4 0.4>
    gravityOverride {z=-1000}
    physicsBodyDefinitions [{
      boundBone {boneName=dyn_C_back_coat_4}
      boxExtents=<15 15 15>
      constraintSetup {angularLimitsMax=<10 20 0>, angularLimitsMin=<-10 -20 -15>, angularTarget=<1 0 0>}
      localJointOffset=<2 0 0>
     }]
   }
   n26 = Anim Dynamics - Bone: dyn_C_back_coat_5!(SimulationSpace:AnimPhysSimSpaceType Enum, ComponentPose=@n25, Alpha=1, GravityOverride:Vector="0,0,-1000", ExternalForce="0,0,0") -> (Pose) {
    angularDampingOverride=0.999
    angularSpringConstant=50
    bAngularSpring=true
    bOverrideAngularDamping=true
    bUseGravityOverride=true
    gravityScale=0.35
    simulationSpace=RootRelative
    alphaBoolBlend {blendOption=Linear}
    boundBone {boneName=dyn_C_back_coat_5}
    componentLinearAccScale=<0.4 0.4 0.4>
    componentLinearVelScale=<0.4 0.4 0.4>
    gravityOverride {z=-1000}
    physicsBodyDefinitions [{
      boundBone {boneName=dyn_C_back_coat_5}
      boxExtents=<15 15 15>
      constraintSetup {angularLimitsMax=<10 20 0>, angularLimitsMin=<-10 -20 -15>, angularTarget=<1 0 0>}
      localJointOffset=<2 0 0>
     }]
   }
   n27 = "Transform (Modify) Bone - Bone: dyn_L_front_coat_1"!(ComponentPose=@n26, Alpha=1) -> (Pose) {
    rotationMode=BMM_Additive
    rotationSpace=BCS_BoneSpace
    alphaBoolBlend {blendOption=Linear}
    boneToModify {boneName=dyn_L_front_coat_1}
    rotation {pitch=-5}
   }
   n28 = "Transform (Modify) Bone - Bone: dyn_R_front_coat_1"!(ComponentPose=@n27, Alpha=1) -> (Pose) {
    rotationMode=BMM_Additive
    rotationSpace=BCS_BoneSpace
    alphaBoolBlend {blendOption=Linear}
    boneToModify {boneName=dyn_R_front_coat_1}
    rotation {pitch=-5}
   }
   n29 = "Transform (Modify) Bone - Bone: dyn_L_front_coat_2"!(ComponentPose=@n28, Alpha=1) -> (Pose) {
    rotationMode=BMM_Additive
    rotationSpace=BCS_BoneSpace
    alphaBoolBlend {blendOption=Linear}
    boneToModify {boneName=dyn_L_front_coat_2}
    rotation {pitch=-5}
   }
   n30 = "Transform (Modify) Bone - Bone: dyn_R_front_coat_2"!(ComponentPose=@n29, Alpha=1) -> (Pose) {
    rotationMode=BMM_Additive
    rotationSpace=BCS_BoneSpace
    alphaBoolBlend {blendOption=Linear}
    boneToModify {boneName=dyn_R_front_coat_2}
    rotation {pitch=-5}
   }
   n31 = "Transform (Modify) Bone - Bone: dyn_L_front_coat_3"!(ComponentPose=@n30, Alpha=1) -> (Pose) {
    rotationMode=BMM_Additive
    rotationSpace=BCS_BoneSpace
    alphaBoolBlend {blendOption=Linear}
    boneToModify {boneName=dyn_L_front_coat_3}
    rotation {pitch=-5}
   }
   n32 = "Transform (Modify) Bone - Bone: dyn_R_front_coat_3"!(ComponentPose=@n31, Alpha=1) -> (Pose) {
    rotationMode=BMM_Additive
    rotationSpace=BCS_BoneSpace
    alphaBoolBlend {blendOption=Linear}
    boneToModify {boneName=dyn_R_front_coat_3}
    rotation {pitch=-5}
   }
   n33 = "Transform (Modify) Bone - Bone: dyn_L_side_coat_1"!(ComponentPose=@n32, Alpha=1) -> (Pose) {
    rotationMode=BMM_Additive
    rotationSpace=BCS_BoneSpace
    alphaBoolBlend {blendOption=Linear}
    boneToModify {boneName=dyn_L_side_coat_1}
    rotation {pitch=-10, yaw=-15}
   }
   n34 = "Transform (Modify) Bone - Bone: dyn_C_back_coat_1"!(ComponentPose=@n33, Alpha=1) -> (Pose) {
    rotationMode=BMM_Additive
    rotationSpace=BCS_BoneSpace
    alphaBoolBlend {blendOption=Linear}
    boneToModify {boneName=dyn_C_back_coat_1}
    rotation {yaw=-10}
   }
   n35 = "Transform (Modify) Bone - Bone: dyn_R_side_coat_1"!(ComponentPose=@n34, Alpha=1) -> (Pose) {
    rotationMode=BMM_Additive
    rotationSpace=BCS_BoneSpace
    alphaBoolBlend {blendOption=Linear}
    boneToModify {boneName=dyn_R_side_coat_1}
    rotation {pitch=-10, yaw=-15}
   }
   n36 = "Transform (Modify) Bone - Bone: dyn_L_side_coat_2"!(ComponentPose=@n35, Alpha=1) -> (Pose) {
    rotationMode=BMM_Additive
    rotationSpace=BCS_BoneSpace
    alphaBoolBlend {blendOption=Linear}
    boneToModify {boneName=dyn_L_side_coat_2}
    rotation {pitch=-10, yaw=-15}
   }
   n37 = "Transform (Modify) Bone - Bone: dyn_C_back_coat_2"!(ComponentPose=@n36, Alpha=1) -> (Pose) {
    rotationMode=BMM_Additive
    rotationSpace=BCS_BoneSpace
    alphaBoolBlend {blendOption=Linear}
    boneToModify {boneName=dyn_C_back_coat_2}
    rotation {yaw=-10}
   }
   n38 = "Transform (Modify) Bone - Bone: dyn_R_side_coat_2"!(ComponentPose=@n37, Alpha=1) -> (Pose) {
    rotationMode=BMM_Additive
    rotationSpace=BCS_BoneSpace
    alphaBoolBlend {blendOption=Linear}
    boneToModify {boneName=dyn_R_side_coat_2}
    rotation {pitch=-10, yaw=-15}
   }
   n39 = "Transform (Modify) Bone - Bone: dyn_L_side_coat_3"!(ComponentPose=@n38, Alpha=1) -> (Pose) {
    rotationMode=BMM_Additive
    rotationSpace=BCS_BoneSpace
    alphaBoolBlend {blendOption=Linear}
    boneToModify {boneName=dyn_L_side_coat_3}
    rotation {pitch=-10, yaw=-15}
   }
   n40 = "Transform (Modify) Bone - Bone: dyn_C_back_coat_3"!(ComponentPose=@n39, Alpha=1) -> (Pose) {
    rotationMode=BMM_Additive
    rotationSpace=BCS_BoneSpace
    alphaBoolBlend {blendOption=Linear}
    boneToModify {boneName=dyn_C_back_coat_3}
    rotation {yaw=-10}
   }
   n41 = "Transform (Modify) Bone - Bone: dyn_R_side_coat_3"!(ComponentPose=@n40, Alpha=1) -> (Pose) {
    rotationMode=BMM_Additive
    rotationSpace=BCS_BoneSpace
    alphaBoolBlend {blendOption=Linear}
    boneToModify {boneName=dyn_R_side_coat_3}
    rotation {pitch=-10, yaw=-15}
   }
   n42 = "Transform (Modify) Bone - Bone: dyn_L_front_coat_1"(ComponentPose=@n41, Alpha=1) {
    scaleMode=BMM_Additive
    scaleSpace=BCS_BoneSpace
    alphaBoolBlend {blendOption=Linear}
    boneToModify {boneName=dyn_L_front_coat_1}
    scale {x=0.6}
   }
   n43 = "Transform (Modify) Bone - Bone: dyn_L_side_coat_1"(ComponentPose=@n42, Alpha=1) {
    scaleMode=BMM_Additive
    scaleSpace=BCS_BoneSpace
    alphaBoolBlend {blendOption=Linear}
    boneToModify {boneName=dyn_L_side_coat_1}
    scale {x=0.6}
   }
   n44 = "Transform (Modify) Bone - Bone: dyn_L_front_coat_2"(ComponentPose=@n43, Alpha=1) {
    scaleMode=BMM_Additive
    scaleSpace=BCS_BoneSpace
    alphaBoolBlend {blendOption=Linear}
    boneToModify {boneName=dyn_L_front_coat_2}
    scale {x=0.6}
   }
   n45 = "Transform (Modify) Bone - Bone: dyn_L_side_coat_2"(ComponentPose=@n44, Alpha=1) {
    scaleMode=BMM_Additive
    scaleSpace=BCS_BoneSpace
    alphaBoolBlend {blendOption=Linear}
    boneToModify {boneName=dyn_L_side_coat_2}
    scale {x=0.6}
   }
   n46 = "Transform (Modify) Bone - Bone: dyn_R_front_coat_1"(ComponentPose=@n45, Alpha=1) {
    scaleMode=BMM_Additive
    scaleSpace=BCS_BoneSpace
    alphaBoolBlend {blendOption=Linear}
    boneToModify {boneName=dyn_R_front_coat_1}
    scale {x=0.6}
   }
   n47 = "Transform (Modify) Bone - Bone: dyn_R_side_coat_1"(ComponentPose=@n46, Alpha=1) {
    scaleMode=BMM_Additive
    scaleSpace=BCS_BoneSpace
    alphaBoolBlend {blendOption=Linear}
    boneToModify {boneName=dyn_R_side_coat_1}
    scale {x=0.6}
   }
   n48 = "Transform (Modify) Bone - Bone: dyn_R_front_coat_2"(ComponentPose=@n47, Alpha=1) {
    scaleMode=BMM_Additive
    scaleSpace=BCS_BoneSpace
    alphaBoolBlend {blendOption=Linear}
    boneToModify {boneName=dyn_R_front_coat_2}
    scale {x=0.6}
   }
   n49 = "Transform (Modify) Bone - Bone: dyn_R_side_coat_2"(ComponentPose=@n48, Alpha=1) {
    scaleMode=BMM_Additive
    scaleSpace=BCS_BoneSpace
    alphaBoolBlend {blendOption=Linear}
    boneToModify {boneName=dyn_R_side_coat_2}
    scale {x=0.6}
   }
   n50 = "Transform (Modify) Bone - Bone: dyn_C_back_coat_1"(ComponentPose=@n49, Alpha=1) {
    scaleMode=BMM_Additive
    scaleSpace=BCS_BoneSpace
    alphaBoolBlend {blendOption=Linear}
    boneToModify {boneName=dyn_C_back_coat_1}
    scale {x=0.6}
   }
   n51 = "Transform (Modify) Bone - Bone: dyn_C_back_coat_2"(ComponentPose=@n50, Alpha=1) {
    scaleMode=BMM_Additive
    scaleSpace=BCS_BoneSpace
    alphaBoolBlend {blendOption=Linear}
    boneToModify {boneName=dyn_C_back_coat_2}
    scale {x=0.6}
   }
   n52 = "Constraint - dyn_L_side_coat_2 (PP)"!(ComponentPose=@n51, Alpha=1, ConstraintWeights_0=0.5, ConstraintWeights_1:float=0.5) -> (Pose) {
    alphaBoolBlend {blendOption=Linear}
    alphaScaleBiasClamp {scale=0.2}
    boneToModify {boneName=dyn_L_side_coat_2}
    constraintSetup [{targetBone {boneName=dyn_L_front_coat_2}}, {targetBone {boneName=dyn_C_back_coat_2}}]
    constraintWeights [1,1]
   }
   n53 = "Constraint - dyn_L_side_coat_3 (PP)"!(ComponentPose=@n52, Alpha=1, ConstraintWeights_0=0.5, ConstraintWeights_1:float=0.5) -> (Pose) {
    alphaBoolBlend {blendOption=Linear}
    alphaScaleBiasClamp {scale=0.3}
    boneToModify {boneName=dyn_L_side_coat_3}
    constraintSetup [{targetBone {boneName=dyn_L_front_coat_3}}, {targetBone {boneName=dyn_C_back_coat_3}}]
    constraintWeights [1,1]
   }
   n54 = "Constraint - dyn_L_side_coat_4 (PP)"!(ComponentPose=@n53, Alpha=1, ConstraintWeights_0=0.5, ConstraintWeights_1:float=0.5) -> (Pose) {
    alphaBoolBlend {blendOption=Linear}
    alphaScaleBiasClamp {scale=0.2}
    boneToModify {boneName=dyn_L_side_coat_4}
    constraintSetup [{targetBone {boneName=dyn_L_front_coat_4}}, {targetBone {boneName=dyn_C_back_coat_4}}]
    constraintWeights [1,1]
   }
   n55 = "Constraint - dyn_L_side_coat_5 (PP)"!(ComponentPose=@n54, Alpha=1, ConstraintWeights_0=0.5, ConstraintWeights_1:float=0.5) -> (Pose) {
    alphaBoolBlend {blendOption=Linear}
    alphaScaleBiasClamp {scale=0.2}
    boneToModify {boneName=dyn_L_side_coat_5}
    constraintSetup [{targetBone {boneName=dyn_L_front_coat_5}}, {targetBone {boneName=dyn_C_back_coat_5}}]
    constraintWeights [1,1]
   }
   n56 = "Constraint - dyn_R_side_coat_2 (PP)"!(ComponentPose=@n55, Alpha=1, ConstraintWeights_0=0.5, ConstraintWeights_1:float=0.5) -> (Pose) {
    alphaBoolBlend {blendOption=Linear}
    alphaScaleBiasClamp {scale=0.2}
    boneToModify {boneName=dyn_R_side_coat_2}
    constraintSetup [{targetBone {boneName=dyn_R_front_coat_2}}, {targetBone {boneName=dyn_C_back_coat_2}}]
    constraintWeights [1,1]
   }
   n57 = "Constraint - dyn_R_side_coat_3 (PP)"!(ComponentPose=@n56, Alpha=1, ConstraintWeights_0=0.5, ConstraintWeights_1:float=0.5) -> (Pose) {
    alphaBoolBlend {blendOption=Linear}
    alphaScaleBiasClamp {scale=0.3}
    boneToModify {boneName=dyn_R_side_coat_3}
    constraintSetup [{targetBone {boneName=dyn_R_front_coat_3}}, {targetBone {boneName=dyn_C_back_coat_3}}]
    constraintWeights [1,1]
   }
   n58 = "Constraint - dyn_R_side_coat_4 (PP)"!(ComponentPose=@n57, Alpha=1, ConstraintWeights_0=0.5, ConstraintWeights_1:float=0.5) -> (Pose) {
    alphaBoolBlend {blendOption=Linear}
    alphaScaleBiasClamp {scale=0.2}
    boneToModify {boneName=dyn_R_side_coat_4}
    constraintSetup [{targetBone {boneName=dyn_R_front_coat_4}}, {targetBone {boneName=dyn_C_back_coat_4}}]
    constraintWeights [1,1]
   }
   n59 = "Constraint - dyn_R_side_coat_5 (PP)"!(ComponentPose=@n58, Alpha=1, ConstraintWeights_0=0.5, ConstraintWeights_1:float=0.5) -> (Pose) {
    alphaBoolBlend {blendOption=Linear}
    alphaScaleBiasClamp {scale=0.2}
    boneToModify {boneName=dyn_R_side_coat_5}
    constraintSetup [{targetBone {boneName=dyn_R_front_coat_5}}, {targetBone {boneName=dyn_C_back_coat_5}}]
    constraintWeights [1,1]
   }
   n60 = "Constraint - dyn_C_back_coat_2 (PP)"!(ComponentPose=@n59, Alpha=1, ConstraintWeights_0=0.5, ConstraintWeights_1:float=0.5) -> (Pose) {
    alphaBoolBlend {blendOption=Linear}
    alphaScaleBiasClamp {scale=0.2}
    boneToModify {boneName=dyn_C_back_coat_2}
    constraintSetup [{targetBone {boneName=dyn_L_side_coat_2}}, {targetBone {boneName=dyn_R_side_coat_2}}]
    constraintWeights [1,1]
   }
   n61 = "Constraint - dyn_C_back_coat_3 (PP)"!(ComponentPose=@n60, Alpha=1, ConstraintWeights_0=0.5, ConstraintWeights_1:float=0.5) -> (Pose) {
    alphaBoolBlend {blendOption=Linear}
    alphaScaleBiasClamp {scale=0.3}
    boneToModify {boneName=dyn_C_back_coat_3}
    constraintSetup [{targetBone {boneName=dyn_L_side_coat_3}}, {targetBone {boneName=dyn_R_side_coat_3}}]
    constraintWeights [1,1]
   }
   n62 = "Constraint - dyn_C_back_coat_4 (PP)"!(ComponentPose=@n61, Alpha=1, ConstraintWeights_0=0.5, ConstraintWeights_1:float=0.5) -> (Pose) {
    alphaBoolBlend {blendOption=Linear}
    alphaScaleBiasClamp {scale=0.2}
    boneToModify {boneName=dyn_C_back_coat_4}
    constraintSetup [{targetBone {boneName=dyn_L_side_coat_4}}, {targetBone {boneName=dyn_R_side_coat_4}}]
    constraintWeights [1,1]
   }
   n63 = "Constraint - dyn_C_back_coat_5 (PP)"!(ComponentPose=@n62, Alpha=1, ConstraintWeights_0=0.5, ConstraintWeights_1:float=0.5) -> (Pose) {
    alphaBoolBlend {blendOption=Linear}
    alphaScaleBiasClamp {scale=0.2}
    boneToModify {boneName=dyn_C_back_coat_5}
    constraintSetup [{targetBone {boneName=dyn_L_side_coat_5}}, {targetBone {boneName=dyn_R_side_coat_5}}]
    constraintWeights [1,1]
   }
   n64 = "Transform (Modify) Bone - Bone: dyn_L_thumb_scale"!(ComponentPose=@n63, Alpha=1, Rotation="0,0,0") -> (Pose) {
    scaleMode=BMM_Additive
    scaleSpace=BCS_BoneSpace
    translationMode=BMM_Additive
    translationSpace=BCS_BoneSpace
    alphaBoolBlend {blendOption=Linear}
    boneToModify {boneName=dyn_L_thumb_scale}
    scale=<0.5 0.8 0.5>
    translation {x=-1.5}
   }
   n65 = "Transform (Modify) Bone - Bone: dyn_L_index_scale"!(ComponentPose=@n64, Alpha=1, Rotation="0,0,0") -> (Pose) {
    scaleMode=BMM_Additive
    scaleSpace=BCS_BoneSpace
    translationMode=BMM_Additive
    translationSpace=BCS_BoneSpace
    alphaBoolBlend {blendOption=Linear}
    boneToModify {boneName=dyn_L_index_scale}
    scale=<0.5 0.8 0.5>
    translation {x=-1.5}
   }
   n66 = "Transform (Modify) Bone - Bone: dyn_L_middle_scale"!(ComponentPose=@n65, Alpha=1, Rotation="0,0,0") -> (Pose) {
    scaleMode=BMM_Additive
    scaleSpace=BCS_BoneSpace
    translationMode=BMM_Additive
    translationSpace=BCS_BoneSpace
    alphaBoolBlend {blendOption=Linear}
    boneToModify {boneName=dyn_L_middle_scale}
    scale=<0.5 0.8 0.5>
    translation {x=-1.5}
   }
   n67 = "Transform (Modify) Bone - Bone: dyn_L_ring_scale"!(ComponentPose=@n66, Alpha=1, Rotation="0,0,0") -> (Pose) {
    scaleMode=BMM_Additive
    scaleSpace=BCS_BoneSpace
    translationMode=BMM_Additive
    translationSpace=BCS_BoneSpace
    alphaBoolBlend {blendOption=Linear}
    boneToModify {boneName=dyn_L_ring_scale}
    scale=<0.5 0.8 0.5>
    translation {x=-1.5}
   }
   n68 = "Transform (Modify) Bone - Bone: dyn_L_pinky_scale"!(ComponentPose=@n67, Alpha=1, Rotation="0,0,0") -> (Pose) {
    scaleMode=BMM_Additive
    scaleSpace=BCS_BoneSpace
    translationMode=BMM_Additive
    translationSpace=BCS_BoneSpace
    alphaBoolBlend {blendOption=Linear}
    boneToModify {boneName=dyn_L_pinky_scale}
    scale=<0.5 0.8 0.5>
    translation {x=-1.5}
   }
   n69 = "Transform (Modify) Bone - Bone: dyn_R_thumb_scale"!(ComponentPose=@n68, Alpha=1, Rotation="0,0,0") -> (Pose) {
    scaleMode=BMM_Additive
    scaleSpace=BCS_BoneSpace
    translationMode=BMM_Additive
    translationSpace=BCS_BoneSpace
    alphaBoolBlend {blendOption=Linear}
    boneToModify {boneName=dyn_R_thumb_scale}
    scale=<0.5 0.8 0.5>
    translation {x=1.5}
   }
   n70 = "Transform (Modify) Bone - Bone: dyn_R_index_scale"!(ComponentPose=@n69, Alpha=1, Rotation="0,0,0") -> (Pose) {
    scaleMode=BMM_Additive
    scaleSpace=BCS_BoneSpace
    translationMode=BMM_Additive
    translationSpace=BCS_BoneSpace
    alphaBoolBlend {blendOption=Linear}
    boneToModify {boneName=dyn_R_index_scale}
    scale=<0.5 0.8 0.5>
    translation {x=1.5}
   }
   n71 = "Transform (Modify) Bone - Bone: dyn_R_middle_scale"!(ComponentPose=@n70, Alpha=1, Rotation="0,0,0") -> (Pose) {
    scaleMode=BMM_Additive
    scaleSpace=BCS_BoneSpace
    translationMode=BMM_Additive
    translationSpace=BCS_BoneSpace
    alphaBoolBlend {blendOption=Linear}
    boneToModify {boneName=dyn_R_middle_scale}
    scale=<0.5 0.8 0.5>
    translation {x=1.5}
   }
   n72 = "Transform (Modify) Bone - Bone: dyn_R_ring_scale"!(ComponentPose=@n71, Alpha=1, Rotation="0,0,0") -> (Pose) {
    scaleMode=BMM_Additive
    scaleSpace=BCS_BoneSpace
    translationMode=BMM_Additive
    translationSpace=BCS_BoneSpace
    alphaBoolBlend {blendOption=Linear}
    boneToModify {boneName=dyn_R_ring_scale}
    scale=<0.5 0.8 0.5>
    translation {x=1.5}
   }
   n73 = "Transform (Modify) Bone - Bone: dyn_R_pinky_scale"!(ComponentPose=@n72, Alpha=1, Rotation="0,0,0") -> (Pose) {
    scaleMode=BMM_Additive
    scaleSpace=BCS_BoneSpace
    translationMode=BMM_Additive
    translationSpace=BCS_BoneSpace
    alphaBoolBlend {blendOption=Linear}
    boneToModify {boneName=dyn_R_pinky_scale}
    scale=<0.5 0.8 0.5>
    translation {x=1.5}
   }
   Output Pose(Result=Component To Local(ComponentPose=@n73)) {layerGroup=None}
  ]}
 EventGraph {nodes [
   Event BlueprintUpdateAnimation()
   > n1 = Sequence!(execute) -> (then_0, then_1, then_2:Exec)
   > Set simSpace(simSpace:AnimPhysSimSpaceType Enum) -> (Output_Get:AnimPhysSimSpaceType Enum)
   > Branch(Condition=OR Boolean(A=Get bIsDBNO(), B=Get bIsInFrontEnd()))
   > n2 = Branch(Condition=OR Boolean(A=Get bIsDBNOCarried(), B=Get bIsInFrontEnd()))
   > Set simSpace(simSpace=World)
   Set simSpace(execute=@n2.else)
   Set frontendAlpha(execute=@n1.then_1, frontendAlpha:double=SelectFloat(A=0.5, B=1, bPickA=Get bIsInFrontEnd())) -> (Output_Get:double)
   > n3 = PlayerMovement(AnimInstance=Self-Reference(), XDotProductCoefficient=25, YDotProductCoefficient=25, ZDotProductCoefficient=25)
   > n4 = Set MovementAlpha(MovementAlpha:double="Absolute (Float)"(A=float * float(A=@n3.YMovement, B=0.045) {is=title K2Node_PromotableOperator})) -> (Output_Get:double)
   > Set MovementAlpha(MovementAlpha=float - float(A=@n4, B=Get customDarkStanceGliderDeployed() -> (customDarkStanceGliderDeployed:double)))
   > n5 = Vehicle Dynamics Scaler Function!(execute, CharacterPartAnimInstance:Custom Character Part Anim Instance Object Reference=Self-Reference(), Debug:bool=false) -> (then, InVehicleHeadScaleAlpha:double, InVehicleHeadSimAlpha:double, InVehicleSkirtScaleAlpha:double, InVehicleSkirtSimAlpha:double)
   > Set In Vehicle Skirt Scale Alpha(In Vehicle Skirt Scale Alpha:double=@n5.InVehicleSkirtScaleAlpha) -> (Output_Get:double)
   > n6 = Gravity Override State Function DEPRECATED(CharacterPartAnimInstance=Self-Reference(), GameTimeInSeconds=GetGameTimeInSeconds(), GravityOverrideParams=Get Gravity Override Params())
   > Set GravityOverride(GravityOverride=@n6)
   Branch(execute=@n1.then_2, Condition=Get bIsInFrontEnd())
   > Set isPlayingFrontendIdle(isPlayingFrontendIdle:float=GetCurveValue(self=GetOwnerFrontEndAnimInstance(), CurveName=Frontend_Idle_Bypass)) -> (Output_Get:float)
   n7 = Event OnParachuteOpened()
   n8 = Make Array("[0]"=(TagName="Cosmetics.MetaTags.DarkStance"))
   Set customDarkStanceGliderDeployed(
    execute=[@n7.then, @n9.then]
    customDarkStanceGliderDeployed:double="To Float (Boolean)"(InBool=GliderItemDefinitionHasAnyMetaTag!(Tags:Array of Gameplay Tag Structures=@n8, bRequiresOpened:bool=true) -> (ReturnValue:bool))
   ) -> (Output_Get:double)
   n9 = Event OnParachuteClosed()
  ]}
]
variables [
 customDarkStanceGliderDeployed {type=number}
 frontendAlpha {type=number}
 Gravity Override Params {
  type=GravityOverrideParamsStruct
  default {
   onGround_Moving {gravityOverride=<0 0 -700>}
   onGround_Standing {gravityOverride=<0 0 -700>}
   skydive_Parachuting {gravityOverride=<0 -100 -980>, windAmplitude=<50 -50 50>, windFrequency=<8 8 8>}
   skydive_Surfing {gravityOverride=<0 -100 -980>, windAmplitude=<100 -100 100>, windFrequency=<8 8 8>}
  }
 }
 GravityOverride {type=Vector}
 In Vehicle Skirt Scale Alpha {type=number}
 isPlayingFrontendIdle {type=number}
 MovementAlpha {type=number}
 simSpace {type=AnimPhysSimSpaceType}
]
